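# Conversation database for The Autumn Queen NPC, giving her greeting dialogue in the NPC library.
# A single entry point leads to a Speech node with a default line about the Island's fading life force.
# Contextual lines override it for a first-ever conversation or specific player outfits matched by character ID, ranked through priority curve rows.
# The speech then branches through link nodes into the shared NPC library Quests and Basic Services conversations.
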ConversationDatabase /NPCLibrary/NPCs/TheAutumnQueen/NPCLibrary_Conversation_TheAutumnQueen
entry_points [{tag=Conversation.Entry.AutumnQueen, destinations [#1]}]
nodes [
 ConversationEntryPointNode_7 {
  id=#1
  class=ConversationEntryPointNode
  outputs [#2]
  properties {entryTag {tagName=Conversation.Entry.AutumnQueen}}
 }
 Speech_C_4 {
  id=#2
  class=Speech_C
  outputs [#3, #4]
  properties {
   nodeName=""
   generalConfig {
    defaultMessage=The life force holding this Island together is fading. I do not know what our future holds.
    contextualMessages [
     {
      message="Hail and well met, traveller, and welcome to the Autumnwood."
      contextRequirements [{requirement {class=FortControllerRequirement_IsFirstEverConversationWithNPC}}]
      priority {curve {rowName=Tandem.Speech.Priority.FirstEverInteraction}}
     }
     {
      message="My golden knight. Protect the travellers of The Autumnwood, as is your oath."
      contextRequirements [{requirement {class=FortControllerRequirement_HasCID, properties {allowedCIds [/BRCosmetics/Athena/Items/Cosmetics/Characters/Character_GoldenGuard]}}}]
      priority {curve {rowName=Tandem.Speech.Priority.OutfitReaction}}
     }
     {
      message="You must move quickly, thief. Time is running short."
      contextRequirements [{requirement {class=FortControllerRequirement_HasCID, properties {allowedCIds [&Character_MusketSlinger]}}}]
      priority {curve {rowName=Tandem.Speech.Priority.OutfitReaction}}
     }
     {
      message="I can see your heart. Your intentions are not pure, but they are noble."
      contextRequirements [{requirement {class=FortControllerRequirement_HasCID, properties {allowedCIds [&Character_InnovatorSand]}}}]
      priority {curve {rowName=Tandem.Speech.Priority.OutfitReaction}}
     }
     {
      message=The future is in your hands. Don't drop it.
      contextRequirements [{requirement {
         class=FortControllerRequirement_HasCID
         properties {allowedCIds [&CID_989_Athena_Commando_M_ProgressiveJonesy]}
        }}]
      priority {curve {rowName=Tandem.Speech.Priority.OutfitReaction}}
     }
     {
      message="Friend, I think you are very lost or very early."
      contextRequirements [{requirement {
         class=FortControllerRequirement_HasCID
         properties {allowedCIds [/BRCosmetics/Athena/Items/Cosmetics/Characters/CID_047_Athena_Commando_F_HolidayReindeer]}
        }}]
      priority {curve {rowName=Tandem.Speech.Priority.OutfitReaction}}
     }
    ]
   }
  }
 }
 ConversationLinkNode_0 {id=#4, class=ConversationLinkNode, properties {remoteEntryTag {tagName=Conversation.Entry.BasicServices}}}
 ConversationLinkNode_1 {id=#3, class=ConversationLinkNode, properties {remoteEntryTag {tagName=Conversation.Entry.NPCLibrary.Quests}}}
]
